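# Post-pose AnimBlueprint for the SoloSnoozeBunny female face accessory, adding secondary motion to hood, bunny ears and bangs.
# It blends the parent pose with shoulder and chest bones from Body_Mesh, cached on customization load, then runs the shared SoloSnooze Control Rig.
# Chained root-relative AnimDynamics springs (wind on side hood bones) feed a RigidBody simulation in spine_05 space.
# An open parachute drives a rig input and additive lace-bone offsets.
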AnimBlueprint /BRCosmetics/Characters/Player/Female/Medium/Bodies/F_MED_SoloSnoozeBunny/Meshes/Parts/F_MED_SoloSnoozeBunny_FaceAcc_AnimBP
target_skeleton=&/F_MED_SoloSnoozeBunny_FaceAcc_Skeleton
graphs [
 AnimGraph {nodes [
   n1 = Control Rig!(Source=Use cached pose 'StartCache'(), Alpha, parachute_opened:bool=Get bIsParachuteOpened(), emote:bool=false) -> (Pose) {controlRigAssetReference {blueprintRigClass=/BRCosmetics/Characters/Player/Female/Medium/Bodies/F_MED_SoloSnooze/Meshes/Parts/F_MED_SoloSnooze_FaceAcc_CtrlRig.F_MED_SoloSnooze_FaceAcc_CtrlRig_C}}
   n2 = Anim Dynamics - Bone: dyn_hood_l(ComponentPose=Local To Component(LocalPose=@n1), Alpha=1, ExternalForce="0,0,0") {
    angularDampingOverride=0.999
    angularSpringConstant=350
    bAngularSpring=true
    bEnableWind=true
    bOverrideAngularDamping=true
    bUseGravityOverride=true
    gravityScale=0.35
    simulationSpace=RootRelative
    windScale=0.4
    alphaBoolBlend {blendOption=Linear}
    boundBone {boneName=dyn_hood_l}
    componentAppliedLinearAccClamp=<3000 3000 3000>
    componentLinearAccScale=<0.9 0.9 0.9>
    componentLinearVelScale=<0.9 0.9 0.9>
    gravityOverride {z=-980}
    physicsBodyDefinitions [{
      boundBone {boneName=dyn_hood_l}
      boxExtents=<20 20 20>
      constraintSetup {angularLimitsMax=<25 60 30>, angularLimitsMin=<-25 -60 -30>, angularTarget=<1 0 0>}
      localJointOffset=<5 0 0>
     }]
    relativeSpaceBone {boneName=head}
   }
   n3 = Anim Dynamics - Bone: dyn_hood_c(ComponentPose=@n2, Alpha=1, ExternalForce="0,0,0") {
    angularDampingOverride=0.999
    angularSpringConstant=200
    bAngularSpring=true
    bOverrideAngularDamping=true
    bUseGravityOverride=true
    gravityScale=0.35
    simulationSpace=RootRelative
    alphaBoolBlend {blendOption=Linear}
    boundBone {boneName=dyn_hood_c}
    componentAppliedLinearAccClamp=<3000 3000 3000>
    componentLinearAccScale=<0.8 0.8 0.8>
    componentLinearVelScale=<0.8 0.8 0.8>
    gravityOverride {z=-980}
    physicsBodyDefinitions [{
      boundBone {boneName=dyn_hood_c}
      boxExtents=<15 12 12>
      constraintSetup {angularLimitsMax=<15 15 15>, angularLimitsMin=<-15 -15 -15>, angularTarget=<1 0 0>}
      localJointOffset=<2.5 0 0>
     }]
    relativeSpaceBone {boneName=head}
   }
   n4 = Anim Dynamics - Bone: dyn_hood_r(ComponentPose=@n3, Alpha=1, ExternalForce="0,0,0") {
    angularDampingOverride=0.999
    angularSpringConstant=400
    bAngularSpring=true
    bEnableWind=true
    bOverrideAngularDamping=true
    bUseGravityOverride=true
    gravityScale=0.35
    simulationSpace=RootRelative
    windScale=0.7
    alphaBoolBlend {blendOption=Linear}
    boundBone {boneName=dyn_hood_r}
    componentAppliedLinearAccClamp=<3000 3000 3000>
    componentLinearAccScale=<0.9 0.9 0.9>
    componentLinearVelScale=<0.9 0.9 0.9>
    gravityOverride {z=-980}
    physicsBodyDefinitions [{
      boundBone {boneName=dyn_hood_r}
      boxExtents=<20 20 20>
      constraintSetup {angularLimitsMax=<25 60 30>, angularLimitsMin=<-25 -60 -30>, angularTarget=<1 0 0>}
      localJointOffset=<5 0 0>
     }]
    relativeSpaceBone {boneName=head}
   }
   n5 = Anim Dynamics - Bone: dyn_bunny_ear_01_l(ComponentPose=@n4, Alpha=1, ExternalForce="0,0,0") {
    angularDampingOverride=0.9999
    angularSpringConstant=10000
    bAngularSpring=true
    bOverrideAngularDamping=true
    bUseGravityOverride=true
    gravityScale=0.35
    simulationSpace=RootRelative
    boundBone {boneName=dyn_bunny_ear_01_l}
    componentAppliedLinearAccClamp=<3000 3000 3000>
    componentLinearAccScale=<1 1 1>
    componentLinearVelScale=<1 1 1>
    gravityOverride {z=-980}
    physicsBodyDefinitions [{
      boundBone {boneName=dyn_bunny_ear_01_l}
      boxExtents=<23 23 23>
      constraintSetup {angularLimitsMax=<5 0 15>, angularLimitsMin=<-5 -15 -15>, angularTarget=<1 0 0>}
      localJointOffset=<1 0 0>
     }]
   }
   n6 = Anim Dynamics - Bone: dyn_bunny_ear_02_l(ComponentPose=@n5, Alpha=1, ExternalForce="0,0,0") {
    angularDampingOverride=0.9999
    angularSpringConstant=10000
    bAngularSpring=true
    bOverrideAngularDamping=true
    bUseGravityOverride=true
    gravityScale=0.35
    simulationSpace=RootRelative
    boundBone {boneName=dyn_bunny_ear_02_l}
    componentAppliedLinearAccClamp=<3000 3000 3000>
    componentLinearAccScale=<1 1 1>
    componentLinearVelScale=<1 1 1>
    gravityOverride {z=-980}
    physicsBodyDefinitions [{
      boundBone {boneName=dyn_bunny_ear_02_l}
      boxExtents=<23 23 23>
      constraintSetup {angularLimitsMax=<5 5 15>, angularLimitsMin=<-5 -10 -15>, angularTarget=<1 0 0>}
      localJointOffset=<1 0 0>
     }]
   }
   n7 = Anim Dynamics - Bone: dyn_bunny_ear_03_l(ComponentPose=@n6, Alpha=1, ExternalForce="0,0,0") {
    angularDampingOverride=0.9999
    angularSpringConstant=10000
    bAngularSpring=true
    bOverrideAngularDamping=true
    bUseGravityOverride=true
    gravityScale=0.35
    simulationSpace=RootRelative
    boundBone {boneName=dyn_bunny_ear_03_l}
    componentAppliedLinearAccClamp=<3000 3000 3000>
    componentLinearAccScale=<1 1 1>
    componentLinearVelScale=<1 1 1>
    gravityOverride {z=-980}
    physicsBodyDefinitions [{
      boundBone {boneName=dyn_bunny_ear_03_l}
      boxExtents=<23 23 23>
      constraintSetup {angularLimitsMax=<5 0 15>, angularLimitsMin=<-5 -15 -15>, angularTarget=<1 0 0>}
      localJointOffset=<1 0 0>
     }]
   }
   n8 = Anim Dynamics - Bone: dyn_bunny_ear_04_l(ComponentPose=@n7, Alpha=1, ExternalForce="0,0,0") {
    angularDampingOverride=0.9999
    angularSpringConstant=10000
    bAngularSpring=true
    bOverrideAngularDamping=true
    bUseGravityOverride=true
    gravityScale=0.35
    simulationSpace=RootRelative
    boundBone {boneName=dyn_bunny_ear_04_l}
    componentAppliedLinearAccClamp=<3000 3000 3000>
    componentLinearAccScale=<1 1 1>
    componentLinearVelScale=<1 1 1>
    gravityOverride {z=-980}
    physicsBodyDefinitions [{
      boundBone {boneName=dyn_bunny_ear_04_l}
      boxExtents=<23 23 23>
      constraintSetup {angularLimitsMax=<5 0 15>, angularLimitsMin=<-5 -20 -15>, angularTarget=<1 0 0>}
      localJointOffset=<1 0 0>
     }]
   }
   n9 = Anim Dynamics - Bone: dyn_bunny_ear_05_l(ComponentPose=@n8, Alpha=1, ExternalForce="0,0,0") {
    angularDampingOverride=0.9999
    angularSpringConstant=10000
    bAngularSpring=true
    bOverrideAngularDamping=true
    bUseGravityOverride=true
    gravityScale=0.35
    simulationSpace=RootRelative
    boundBone {boneName=dyn_bunny_ear_05_l}
    componentAppliedLinearAccClamp=<3000 3000 3000>
    componentLinearAccScale=<1 1 1>
    componentLinearVelScale=<1 1 1>
    gravityOverride {z=-980}
    physicsBodyDefinitions [{
      boundBone {boneName=dyn_bunny_ear_05_l}
      boxExtents=<23 23 23>
      constraintSetup {angularLimitsMax=<4 0 15>, angularLimitsMin=<-4 -25 -15>, angularTarget=<1 0 0>}
      localJointOffset=<1 0 0>
     }]
   }
   n10 = Anim Dynamics - Bone: dyn_bunny_ear_01_r(ComponentPose=@n9, Alpha=1, ExternalForce="0,0,0") {
    angularDampingOverride=0.9999
    angularSpringConstant=10000
    bAngularSpring=true
    bOverrideAngularDamping=true
    bUseGravityOverride=true
    gravityScale=0.35
    simulationSpace=RootRelative
    boundBone {boneName=dyn_bunny_ear_01_r}
    componentAppliedLinearAccClamp=<3000 3000 3000>
    componentLinearAccScale=<1 1 1>
    componentLinearVelScale=<1 1 1>
    gravityOverride {z=-980}
    physicsBodyDefinitions [{
      boundBone {boneName=dyn_bunny_ear_01_r}
      boxExtents=<23 23 23>
      constraintSetup {angularLimitsMax=<5 0 15>, angularLimitsMin=<-5 -15 -15>, angularTarget=<1 0 0>}
      localJointOffset=<1 0 0>
     }]
   }
   n11 = Anim Dynamics - Bone: dyn_bunny_ear_02_r(ComponentPose=@n10, Alpha=1, ExternalForce="0,0,0") {
    angularDampingOverride=0.9999
    angularSpringConstant=10000
    bAngularSpring=true
    bOverrideAngularDamping=true
    bUseGravityOverride=true
    gravityScale=0.35
    simulationSpace=RootRelative
    boundBone {boneName=dyn_bunny_ear_02_r}
    componentAppliedLinearAccClamp=<3000 3000 3000>
    componentLinearAccScale=<1 1 1>
    componentLinearVelScale=<1 1 1>
    gravityOverride {z=-980}
    physicsBodyDefinitions [{
      boundBone {boneName=dyn_bunny_ear_02_r}
      boxExtents=<23 23 23>
      constraintSetup {angularLimitsMax=<5 5 15>, angularLimitsMin=<-5 -10 -15>, angularTarget=<1 0 0>}
      localJointOffset=<1 0 0>
     }]
   }
   n12 = Anim Dynamics - Bone: dyn_bunny_ear_03_r(ComponentPose=@n11, Alpha=1, ExternalForce="0,0,0") {
    angularDampingOverride=0.9999
    angularSpringConstant=10000
    bAngularSpring=true
    bOverrideAngularDamping=true
    bUseGravityOverride=true
    gravityScale=0.35
    simulationSpace=RootRelative
    boundBone {boneName=dyn_bunny_ear_03_r}
    componentAppliedLinearAccClamp=<3000 3000 3000>
    componentLinearAccScale=<1 1 1>
    componentLinearVelScale=<1 1 1>
    gravityOverride {z=-980}
    physicsBodyDefinitions [{
      boundBone {boneName=dyn_bunny_ear_03_r}
      boxExtents=<23 23 23>
      constraintSetup {angularLimitsMax=<5 0 15>, angularLimitsMin=<-5 -15 -15>, angularTarget=<1 0 0>}
      localJointOffset=<1 0 0>
     }]
   }
   n13 = Anim Dynamics - Bone: dyn_bunny_ear_04_r(ComponentPose=@n12, Alpha=1, ExternalForce="0,0,0") {
    angularDampingOverride=0.9999
    angularSpringConstant=10000
    bAngularSpring=true
    bOverrideAngularDamping=true
    bUseGravityOverride=true
    gravityScale=0.35
    simulationSpace=RootRelative
    boundBone {boneName=dyn_bunny_ear_04_r}
    componentAppliedLinearAccClamp=<3000 3000 3000>
    componentLinearAccScale=<1 1 1>
    componentLinearVelScale=<1 1 1>
    gravityOverride {z=-980}
    physicsBodyDefinitions [{
      boundBone {boneName=dyn_bunny_ear_04_r}
      boxExtents=<23 23 23>
      constraintSetup {angularLimitsMax=<5 0 15>, angularLimitsMin=<-5 -20 -15>, angularTarget=<1 0 0>}
      localJointOffset=<1 0 0>
     }]
   }
   n14 = Anim Dynamics - Bone: dyn_bunny_ear_05_r(ComponentPose=@n13, Alpha=1, ExternalForce="0,0,0") {
    angularDampingOverride=0.9999
    angularSpringConstant=10000
    bAngularSpring=true
    bOverrideAngularDamping=true
    bUseGravityOverride=true
    gravityScale=0.35
    simulationSpace=RootRelative
    boundBone {boneName=dyn_bunny_ear_05_r}
    componentAppliedLinearAccClamp=<3000 3000 3000>
    componentLinearAccScale=<1 1 1>
    componentLinearVelScale=<1 1 1>
    gravityOverride {z=-980}
    physicsBodyDefinitions [{
      boundBone {boneName=dyn_bunny_ear_05_r}
      boxExtents=<23 23 23>
      constraintSetup {angularLimitsMax=<4 0 15>, angularLimitsMin=<-4 -25 -15>, angularTarget=<1 0 0>}
      localJointOffset=<1 0 0>
     }]
   }
   n15 = Anim Dynamics - Bone: dyn_bang(ComponentPose=@n14, Alpha=1, ExternalForce="0,0,0") {
    angularDampingOverride=0.9
    angularSpringConstant=3500
    bAngularSpring=true
    bOverrideAngularDamping=true
    bUseGravityOverride=true
    gravityScale=0.35
    simulationSpace=RootRelative
    boundBone {boneName=dyn_bang}
    componentAppliedLinearAccClamp=<3000 3000 3000>
    componentLinearAccScale=<0.9 0.9 0.9>
    componentLinearVelScale=<0.9 0.9 0.9>
    gravityOverride {z=-980}
    physicsBodyDefinitions [{
      boundBone {boneName=dyn_bang}
      boxExtents=<24 24 24>
      constraintSetup {angularLimitsMax=<2 15 25>, angularLimitsMin=<-2 -15 0>, angularTarget=<1 0 0>}
      localJointOffset=<2 2 0>
     }]
   }
   n16 = RigidBody(ComponentPose=@n15) {
    bOverrideWorldGravity=true
    overridePhysicsAsset=&/F_MED_SoloSnoozeBunny_FaceAcc_Physics
    simulationSpace=BaseBoneSpace
    baseBoneRef {boneName=spine_05}
    overrideWorldGravity {z=-980}
    simSpaceSettings {dampingAlpha=0, maxAngularAcceleration=2200, maxAngularVelocity=50, maxLinearAcceleration=2200, maxLinearVelocity=65, worldAlpha=1}
   }
   n17 = Get bIsParachuteOpened()
   n18 = "Transform (Modify) Bone - Bone: MSTR_laces_r"!(ComponentPose=@n16, bAlphaBoolEnabled:bool=@n17, Translation="0,0.5,0", Rotation="0,0,-4", Scale="1,1,1") -> (Pose) {
    alphaInputType=Bool
    rotationMode=BMM_Additive
    translationMode=BMM_Additive
    boneToModify {boneName=MSTR_laces_r}
    rotation {roll=-4}
    translation {y=0.5}
   }
   n19 = "Transform (Modify) Bone - Bone: MSTR_laces_l"!(ComponentPose=@n18, bAlphaBoolEnabled:bool=@n17, Translation="0,0.5,0", Rotation="0,0,-4", Scale="1,1,1") -> (Pose) {
    alphaInputType=Bool
    rotationMode=BMM_Additive
    translationMode=BMM_Additive
    boneToModify {boneName=MSTR_laces_l}
    rotation {roll=-4}
    translation {y=0.5}
   }
   Output Pose(Result=Component To Local(ComponentPose=@n19)) {layerGroup=None}
   n20 = Layered blend per bone(
    BasePose=Copy Pose From Mesh() {bCopyCurves=true, bUseAttachedParent=true}
    BlendPoses_0=Copy Pose From Mesh(SourceMeshComponent=Get Body_Mesh()) {bUseAttachedParent=true}
   ) {layerSetup [{branchFilters [
       {boneName=clavicle_l}
       {boneName=clavicle_r}
       {boneName=upperarm_l}
       {boneName=upperarm_r}
       {boneName=lowerarm_r}
       {boneName=deform_pec_aim_base_l}
       {boneName=deform_pec_aim_l}
       {boneName=deform_pec_base_l}
       {boneName=deform_pec_l}
       {boneName=deform_pec_aim_base_r}
       {boneName=deform_pec_aim_r}
       {boneName=deform_pec_base_r}
       {boneName=deform_pec_r}
       {boneName=deform_upperarm_fr_l}
       {boneName=deform_upperarm_up_l}
      ]}]}
   Save cached pose 'StartCache'(Pose=@n20) {cacheName=StartCache}
  ]}
 EventGraph {nodes [
   n1 = Event OnCharacterCustomizationLoaded()
   > Set Body_Mesh(Body_Mesh=Get Array Item with EFort Custom Part Type(Array=@n1.CustomMeshParts))
   > Get Dynamic Material Instance(SkeletalMeshComponent=Get Array Item with EFort Custom Part Type(Array=@n1.CustomMeshParts, PartType=Face))
   Event BlueprintUpdateAnimation()
  ]}
]
variables [Body_Mesh {type=SkeletalMeshComponent}]
