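# Control Rig for the SoloSnoozeBunny face accessory, nudging the left and right hood master bones as the arms move.
# Spherical pose readers on the upper arms weight relative offsets, with one right-side offset multiplying clavicle and upper-arm readers.
# A sequence runs an interpolated offset fed by constant false, arm correctives only while "Parachute Opened" is false, then always-on correctives with debug drawing enabled.
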ControlRigBlueprint /BRCosmetics/Characters/Player/Female/Medium/Bodies/F_MED_SoloSnoozeBunny/Meshes/Parts/F_MED_SoloSnoozeBunny_FaceAcc_CtrlRig
graphs [
 RigVMFunctionLibrary {links [], nodes []}
 RigVMFunction_Sequence_2_ContainedGraph {nodes [
   n1 = RigVMFunctionEntryNode()
   n2 = &RigVMFunction_Sequence(ExecuteContext=@n1.ExecuteContext)
   n3 = &RigVMFunction_Sequence(ExecuteContext=@n2.B)
   RigVMFunctionReturnNode(A=@n2.A, B=@n3.A, C=@n3.B)
  ]}
 RigVMModel {nodes [
   RigUnit_BeginExecution = &RigUnit_BeginExecution()
   RigVMFunction_Sequence_1 = RigVMAggregateNode(ExecuteContext=@RigUnit_BeginExecution.ExecutePin)
   RigVMFunction_ControlFlowBranch = &RigVMFunction_ControlFlowBranch(ExecuteContext=@RigVMFunction_Sequence_1.B, Condition=RigVMVariableNode() {variable="Parachute Opened "}.Value)
   n1 = &RigUnit_SphericalPoseReader(
    ActiveRegionScaleFactors=(PositiveWidth=1,NegativeWidth=1,PositiveHeight=1,NegativeHeight=1)
    ActiveRegionSize=0.1
    Debug=(bDrawDebug=false,bDraw2D=False,bDrawLocalAxes=False,DebugScale=25,DebugSegments=20,DebugThickness=0.25)
    DriverAxis=(X=-1,Y=0,Z=0)
    DriverItem=(Type=Bone,Name="upperarm_r")
    FalloffRegionScaleFactors=(PositiveWidth=1,NegativeWidth=1,PositiveHeight=1,NegativeHeight=1)
    FalloffSize=0.23
    FlipHeightScaling=False
    FlipWidthScaling=False
    OptionalParentItem=(Type=Bone,Name="None")
    RotationOffset=(X=-120,Y=0,Z=29)
    ExecutePin=@RigVMFunction_ControlFlowBranch.False
   )
   > RigUnit_SetRelativeTransformForItem = /Script/ControlRig.RigUnit_SetRelativeTransformForItem(
    Child=(Type=Bone,Name="MSTR_hood_r")
    Parent=(Type=Bone,Name="MSTR_hood_r")
    Value=(Rotation=(X=0,Y=0,Z=-0.2164,W=0.9763),Translation=(X=0,Y=0,Z=11),Scale3D=(X=1,Y=1,Z=1))
    bParentInitial=False
    bPropagateToChildren=True
    Weight=@n1.OutputParam
   ) {
    is=struct RigVMUnitNode
    template="Set Relative Transform::Execute(in Child,in Parent,in bParentInitial,in Value,in Weight,in bPropagateToChildren)"
   }
   > &RigUnit_SetRelativeTransformForItem(
    Child=(Type=Bone,Name="MSTR_hood_l")
    Parent=(Type=Bone,Name="MSTR_hood_l")
    Value=(Rotation=(X=0,Y=0,Z=0,W=1),Translation=(X=0,Y=0,Z=-7),Scale3D=(X=1,Y=1,Z=1))
    bParentInitial=False
    bPropagateToChildren=True
    Weight=@n1.OutputParam
   ) {template="Set Relative Transform::Execute(in Child,in Parent,in bParentInitial,in Value,in Weight,in bPropagateToChildren)"}
   > n2 = &RigUnit_SphericalPoseReader(
    ActiveRegionScaleFactors=(PositiveWidth=1,NegativeWidth=1,PositiveHeight=1,NegativeHeight=1)
    ActiveRegionSize=0.1
    Debug=(bDrawDebug=false,bDraw2D=False,bDrawLocalAxes=False,DebugScale=25,DebugSegments=20,DebugThickness=0.25)
    DriverAxis=(X=1,Y=0,Z=0)
    DriverItem=(Type=Bone,Name="upperarm_l")
    FalloffRegionScaleFactors=(PositiveWidth=1,NegativeWidth=1,PositiveHeight=1,NegativeHeight=1)
    FalloffSize=0.25
    FlipHeightScaling=False
    FlipWidthScaling=False
    OptionalParentItem=(Type=Bone,Name="None")
    RotationOffset=(X=-10,Y=-45,Z=0)
   )
   > &RigUnit_SetRelativeTransformForItem(
    Child=(Type=Bone,Name="MSTR_hood_l")
    Parent=(Type=Bone,Name="MSTR_hood_l")
    Value=(Rotation=(X=0,Y=0,Z=0,W=1),Translation=(X=0,Y=0,Z=-4),Scale3D=(X=1,Y=1,Z=1))
    bParentInitial=False
    bPropagateToChildren=True
    Weight=@n2.OutputParam
   ) {template="Set Relative Transform::Execute(in Child,in Parent,in bParentInitial,in Value,in Weight,in bPropagateToChildren)"}
   > n3 = &RigUnit_SphericalPoseReader(
    ActiveRegionScaleFactors=(PositiveWidth=1,NegativeWidth=1,PositiveHeight=1,NegativeHeight=1)
    ActiveRegionSize=0.1
    Debug=(bDrawDebug=false,bDraw2D=False,bDrawLocalAxes=False,DebugScale=25,DebugSegments=20,DebugThickness=0.25)
    DriverAxis=(X=-1,Y=0,Z=0)
    DriverItem=(Type=Bone,Name="clavicle_r")
    FalloffRegionScaleFactors=(PositiveWidth=1,NegativeWidth=1,PositiveHeight=1,NegativeHeight=1)
    FalloffSize=0.2
    FlipHeightScaling=False
    FlipWidthScaling=False
    OptionalParentItem=(Type=Bone,Name="None")
    RotationOffset=(X=-10,Y=130,Z=0)
   )
   > n4 = &RigUnit_SphericalPoseReader(
    ActiveRegionScaleFactors=(PositiveWidth=1,NegativeWidth=1,PositiveHeight=2,NegativeHeight=2)
    ActiveRegionSize=0.1
    Debug=(bDrawDebug=false,bDraw2D=False,bDrawLocalAxes=False,DebugScale=25,DebugSegments=20,DebugThickness=0.25)
    DriverAxis=(X=-1,Y=0,Z=0)
    DriverItem=(Type=Bone,Name="upperarm_r")
    FalloffRegionScaleFactors=(PositiveWidth=1,NegativeWidth=1,PositiveHeight=1,NegativeHeight=1)
    FalloffSize=0.3
    FlipHeightScaling=False
    FlipWidthScaling=False
    OptionalParentItem=(Type=Bone,Name="None")
    RotationOffset=(X=-100,Y=40,Z=0)
   )
   Multiply = &RigVMFunction_MathDoubleMul(B=@n3.OutputParam, A=@n4.OutputParam)
   > &RigUnit_SetRelativeTransformForItem(
    Child=(Type=Bone,Name="MSTR_hood_r")
    Parent=(Type=Bone,Name="MSTR_hood_r")
    Value=(Rotation=(X=0,Y=0,Z=0,W=1),Translation=(X=-4,Y=0,Z=2),Scale3D=(X=1,Y=1,Z=1))
    bParentInitial=False
    bPropagateToChildren=True
    Weight=@Multiply.Result
   ) {template="Set Relative Transform::Execute(in Child,in Parent,in bParentInitial,in Value,in Weight,in bPropagateToChildren)"}
   n5 = &RigUnit_SphericalPoseReader(
    ActiveRegionScaleFactors=(PositiveWidth=1.5,NegativeWidth=1,PositiveHeight=1.6,NegativeHeight=1.6)
    ActiveRegionSize=0.1
    Debug=(bDrawDebug=true,bDraw2D=False,bDrawLocalAxes=False,DebugScale=25,DebugSegments=20,DebugThickness=0.25)
    DriverAxis=(X=-1,Y=0,Z=0)
    DriverItem=(Type=Bone,Name="upperarm_r")
    FalloffRegionScaleFactors=(PositiveWidth=1,NegativeWidth=1,PositiveHeight=1,NegativeHeight=1)
    FalloffSize=0.3
    FlipHeightScaling=False
    FlipWidthScaling=False
    OptionalParentItem=(Type=Bone,Name="None")
    RotationOffset=(X=-20,Y=145,Z=0)
    ExecutePin=@RigVMFunction_Sequence_1.C
   )
   > n6 = &RigUnit_SphericalPoseReader(
    ActiveRegionScaleFactors=(PositiveWidth=1.5,NegativeWidth=1,PositiveHeight=1.6,NegativeHeight=1.6)
    ActiveRegionSize=0.1
    Debug=(bDrawDebug=true,bDraw2D=False,bDrawLocalAxes=False,DebugScale=25,DebugSegments=20,DebugThickness=0.25)
    DriverAxis=(X=1,Y=0,Z=0)
    DriverItem=(Type=Bone,Name="upperarm_l")
    FalloffRegionScaleFactors=(PositiveWidth=1,NegativeWidth=1,PositiveHeight=1,NegativeHeight=1)
    FalloffSize=0.3
    FlipHeightScaling=False
    FlipWidthScaling=False
    OptionalParentItem=(Type=Bone,Name="None")
    RotationOffset=(X=10,Y=-35,Z=0)
   )
   > &RigUnit_SetRelativeTransformForItem(
    Child=(Type=Bone,Name="MSTR_hood_r")
    Parent=(Type=Bone,Name="MSTR_hood_r")
    Value=(Rotation=(X=0,Y=0,Z=0,W=1),Translation=(X=-2,Y=0,Z=7),Scale3D=(X=1,Y=1,Z=1))
    bParentInitial=False
    bPropagateToChildren=True
    Weight=@n5.OutputParam
   ) {template="Set Relative Transform::Execute(in Child,in Parent,in bParentInitial,in Value,in Weight,in bPropagateToChildren)"}
   > &RigUnit_SetRelativeTransformForItem(
    Child=(Type=Bone,Name="MSTR_hood_l")
    Parent=(Type=Bone,Name="MSTR_hood_l")
    Value=(Rotation=(X=0,Y=0,Z=0,W=1),Translation=(X=-2,Y=0,Z=-4),Scale3D=(X=1,Y=1,Z=1))
    bParentInitial=False
    bPropagateToChildren=True
    Weight=@n6.OutputParam
   ) {template="Set Relative Transform::Execute(in Child,in Parent,in bParentInitial,in Value,in Weight,in bPropagateToChildren)"}
   RigVMFunction_MathBoolToFloat = &RigVMFunction_MathBoolToFloat(Value=False)
   AlphaInterp = &RigVMFunction_AlphaInterp(ClampMin=0.4, InterpSpeedDecreasing=2, InterpSpeedIncreasing=2, Scale=1, bClampResult=false, Value=@RigVMFunction_MathBoolToFloat.Result)
   &RigUnit_SetRelativeTransformForItem(
    Child=(Type=Bone,Name="MSTR_hood_r")
    Parent=(Type=Bone,Name="MSTR_hood_r")
    Value=(Rotation=(X=0,Y=0,Z=0,W=1),Translation=(X=0,Y=0,Z=5.5),Scale3D=(X=1,Y=1,Z=1))
    bParentInitial=False
    bPropagateToChildren=True
    Weight=@AlphaInterp.Result
    ExecutePin=@RigVMFunction_Sequence_1.A
   ) {template="Set Relative Transform::Execute(in Child,in Parent,in bParentInitial,in Value,in Weight,in bPropagateToChildren)"}
   > &RigUnit_SetRelativeTransformForItem(
    Child=(Type=Bone,Name="MSTR_hood_l")
    Parent=(Type=Bone,Name="MSTR_hood_l")
    Value=(Rotation=(X=0,Y=0,Z=0,W=1),Translation=(X=0,Y=0,Z=-5.5),Scale3D=(X=1,Y=1,Z=1))
    bParentInitial=False
    bPropagateToChildren=True
    Weight=@AlphaInterp.Result
   ) {template="Set Relative Transform::Execute(in Child,in Parent,in bParentInitial,in Value,in Weight,in bPropagateToChildren)"}
  ]}
]
hierarchy [
 root {loc=<0 0 0>, rot=<0 0 0 1>}
 pelvis {parent=root, loc=<0 -0.5301 98.0571>, rot=<0.0127 0.707 -0.0127 -0.707>}
 spine_01 {parent=pelvis, loc=<6.1208 -0.9328 0>, rot=<0 0 0.034 -0.9994>}
 spine_02 {parent=spine_01, loc=<6.5997 -0.2095 0>, rot=<0 0 0.0238 -0.9997>}
 spine_03 {parent=spine_02, loc=<6.4914 -0.1237 0>, rot=<0 0 0.0149 -0.9999>}
 spine_04 {parent=spine_03, loc=<6.9598 0.0794 0>, rot=<0 0 -0.0132 -0.9999>}
 spine_05 {parent=spine_04, loc=<6.9987 0 0>, rot=<0 0 -0.0015 -1>}
 neck_01 {parent=spine_05, loc=<11.3138 0.1529 0>, rot=<0 0 -0.2003 -0.9797>}
 neck_02 {parent=neck_01, loc=<4.5627 0 0>, rot=<0 0 0.0504 -0.9987>}
 head {parent=neck_02, loc=<5.3914 0 0>, rot=<0 0 0.115 -0.9934>}
 MSTR_bang {parent=head, loc=<8.0145 8.932 1.0029>, rot=<-0.0307 -0.9404 0.0427 -0.3359>}
 dyn_bang {parent=MSTR_bang, loc=<0 0 0>, rot=<0 0 0 1>}
 MSTR_bunny_ear_l {parent=head, loc=<14.0828 1.6758 -4.3772>, rot=<-0.6077 -0.3615 -0.0174 -0.7069>}
 dyn_bunny_ear_01_l {parent=MSTR_bunny_ear_l, loc=<0 0 0>, rot=<0 0 0 1>}
 dyn_bunny_ear_02_l {parent=dyn_bunny_ear_01_l, loc=<4.4548 0 0>, rot=<0.2579 -0.3292 0.1382 -0.8978>}
 dyn_bunny_ear_03_l {parent=dyn_bunny_ear_02_l, loc=<3.8822 0 0>, rot=<0.3352 -0.4064 -0.1082 -0.843>}
 dyn_bunny_ear_04_l {parent=dyn_bunny_ear_03_l, loc=<5.6571 0 0>, rot=<0.0676 -0.1602 -0.1596 -0.9717>}
 dyn_bunny_ear_05_l {parent=dyn_bunny_ear_04_l, loc=<6.7088 0 0>, rot=<0 0 0 1>}
 MSTR_bunny_ear_r {parent=head, loc=<14.0824 1.6758 4.3772>, rot=<-0.7069 -0.0174 -0.3615 -0.6077>}
 dyn_bunny_ear_01_r {parent=MSTR_bunny_ear_r, loc=<0 0 0>, rot=<0 0 0 1>}
 dyn_bunny_ear_02_r {parent=dyn_bunny_ear_01_r, loc=<4.4548 0 0>, rot=<-0.2579 -0.3292 -0.1382 -0.8978>}
 dyn_bunny_ear_03_r {parent=dyn_bunny_ear_02_r, loc=<3.8822 0 0>, rot=<-0.3352 -0.4064 0.1082 -0.8431>}
 dyn_bunny_ear_04_r {parent=dyn_bunny_ear_03_r, loc=<5.6564 0 0>, rot=<-0.0676 -0.1603 0.1596 -0.9717>}
 dyn_bunny_ear_05_r {parent=dyn_bunny_ear_04_r, loc=<6.7088 0 0>, rot=<-0.4369 -0.0349 0.1485 -0.8865>}
 MSTR_hood_l {parent=neck_01, loc=<1.4925 1.0055 -11.7524>, rot=<-0.1608 -0.987 0 0>}
 dyn_hood_l {parent=MSTR_hood_l, loc=<0 0 0>, rot=<0 0 0 1>}
 MSTR_hood_c {parent=neck_01, loc=<-1.5028 -10.9254 0>, rot=<-0.1608 -0.987 0 0>}
 dyn_hood_c {parent=MSTR_hood_c, loc=<0 0 0>, rot=<0 0 0 1>}
 MSTR_hood_r {parent=neck_01, loc=<1.4927 1.0054 11.7524>, rot=<-0.1608 -0.987 0 0>}
 dyn_hood_r {parent=MSTR_hood_r, loc=<0 0 0>, rot=<0 0 0 1>}
 clavicle_l {parent=spine_05, loc=<5.4831 -0.1053 -3.2946>, rot=<0.0516 -0.7288 -0.0145 -0.6826>}
 upperarm_l {parent=clavicle_l, loc=<9.8658 -1.2136 -0.4105>, rot=<0.0164 -0.3245 -0.0269 -0.9454>}
 lowerarm_l {parent=upperarm_l, loc=<26.465 -0.8345 0.259>, rot=<0.0036 0.0161 -0.2298 -0.9731>}
 upperarm_twist_01_l {parent=upperarm_l, loc=<7.2135 -0.1134 0>, rot=<0 0 0 1>}
 deform_upperarm_fr_base_l {parent=upperarm_twist_01_l, loc=<-5.6312 5.092 0.3354>, rot=<0 0 0 -1>}
 deform_upperarm_fr_l {parent=deform_upperarm_fr_base_l, loc=<0 0 0>, rot=<0 0 0 1>}
 deform_upperarm_up_base_l {parent=upperarm_twist_01_l, loc=<-8.891 0.4825 3.9517>, rot=<0 0 0 -1>}
 deform_upperarm_up_l {parent=deform_upperarm_up_base_l, loc=<0 0 0>, rot=<0 0 0 1>}
 clavicle_r {parent=spine_05, loc=<5.4831 -0.1053 3.2946>, rot=<0.7289 0.0465 -0.6829 0.009>}
 upperarm_r {parent=clavicle_r, loc=<-9.8658 1.2136 0.4105>, rot=<0.0164 -0.3245 -0.0269 -0.9454>}
 lowerarm_r {parent=upperarm_r, loc=<-26.465 0.8345 -0.259>, rot=<0.0036 0.0161 -0.2298 -0.9731>}
 upperarm_twist_01_r {parent=upperarm_r, loc=<-7.2135 0.1134 0>, rot=<0 0 0 -1>}
 deform_upperarm_fr_base_r {parent=upperarm_twist_01_r, loc=<5.6312 -5.092 -0.3354>, rot=<0 0 0 -1>}
 deform_upperarm_fr_r {parent=deform_upperarm_fr_base_r, loc=<0 0 0>, rot=<0 0 0 1>}
 deform_upperarm_up_base_r {parent=upperarm_twist_01_r, loc=<8.891 -0.4825 -3.9517>, rot=<0 0 0 -1>}
 deform_upperarm_up_r {parent=deform_upperarm_up_base_r, loc=<0 0 0>, rot=<0 0 0 1>}
 MSTR_laces_l {parent=spine_05, loc=<3.7867 8.4043 -4.7495>, rot=<-0.1374 -0.6936 -0.657 -0.2614>}
 dyn_laces_01_l {parent=MSTR_laces_l, loc=<0 0 0>, rot=<0 0 0 1>}
 dyn_laces_02_l {parent=dyn_laces_01_l, loc=<5.6767 0 0>, rot=<-0.0006 0.0047 -0.0635 -0.998>}
 dyn_laces_03_l {parent=dyn_laces_02_l, loc=<5.5544 0 0>, rot=<0 -0.1912 0 -0.9815>}
 MSTR_laces_r {parent=spine_05, loc=<3.7867 8.4043 4.7495>, rot=<0.2614 0.657 0.6936 0.1374>}
 dyn_laces_01_r {parent=MSTR_laces_r, loc=<-0.1042 0 0>, rot=<0 0 0 1>}
 dyn_laces_02_r {parent=dyn_laces_01_r, loc=<5.6765 0 0>, rot=<0.0006 0.0047 0.0635 -0.998>}
 dyn_laces_03_r {parent=dyn_laces_02_r, loc=<5.5549 0 0>, rot=<-0.0107 -0.1909 0.055 -0.98>}
 deform_pec_aim_base_l {parent=spine_04, loc=<3.2456 9.3309 -1.2237>, rot=<0.0603 -0.4033 0.3318 -0.8507>}
 deform_pec_aim_l {parent=deform_pec_aim_base_l, loc=<0 0 0>, rot=<-0.0001 0.0004 -0.0002 -1>}
 deform_pec_base_l {parent=deform_pec_aim_l, loc=<9.3665 0 0>, rot=<0 0 0 1>}
 deform_pec_l {parent=deform_pec_base_l, loc=<0.0059 0 0>, rot=<0 0 0 1>}
 deform_pec_aim_base_r {parent=spine_04, loc=<3.2452 9.3309 1.2237>, rot=<-0.4033 -0.0603 0.8507 0.3318>}
 deform_pec_aim_r {parent=deform_pec_aim_base_r, loc=<0 0 0>, rot=<0 0 0 1>}
 deform_pec_base_r {parent=deform_pec_aim_r, loc=<-9.3667 -0.0003 -0.0004>, rot=<0 0 0 1>}
 deform_pec_r {parent=deform_pec_base_r, loc=<0 0 0>, rot=<0 0 0 1>}
]
variables ["Parachute Opened " {type=bool, default=False}]
